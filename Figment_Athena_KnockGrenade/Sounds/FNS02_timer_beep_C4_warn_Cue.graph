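# Sound cue for a warning timer beep on the Figment knock grenade, going by its name and path.
# A single wave player feeds the C4 warning beep wave straight to the output, with no randomization or modulation.
# It uses the gadget-activate attenuation and the gadget items sound class.
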SoundCue /Figment_Athena_KnockGrenade/Sounds/FNS02_timer_beep_C4_warn_Cue
root=WavePlayer_1
nodes [{id=WavePlayer_1, properties {soundWaveAssetPtr=&/timer_beep_C4_warn_edit}}]
properties {
 attenuationSettings=&Fort_GadgetActivate_Att
 maxDistance=4000
 soundClassObject=&Fort_Gadgets_Items
 volumeMultiplier=1
}
